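; Gameplay ability for the Season 23 Battle Pass "damage after throw" quest, which applies a gameplay effect to the player when a triggering event's target qualifies.
; On event activation it calls the parent, then requires the event target to match any tag in an include container and none in an exclude container.
; If both checks pass, it applies the configured effect to the owning pawn's player controller rather than the pawn itself.
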(event Ability|EventActivateAbilityFromEvent
  (params (EventData :type Gameplay%20Event%20Data%20Structure%20%28by%20ref%29))
  (bind (_eventtag _instigator _target) (Utilities|Struct|BreakGameplayEventData EventData))
  (|CallParentFunction EventData)
  (bind _returnvalue (GameplayTags|HasAnyMatchingGameplayTags(Message) _target (Variables|Default|GetInclude%20Tag%20Container)))
  (if _returnvalue
    (bind _returnvalue_2 (GameplayTags|HasAnyMatchingGameplayTags(Message) _target (Variables|Default|GetExclude%20Tag%20Container)))
    (if (not _returnvalue_2)
      (Creative|GameplayEffects|SimpleApplyGameplayEffectToActor (Variables|Default|GetGameplay%20Effect) (Fortnite|GetFortPlayerControllerfromActor (Variables|GAVars|Data|GetOwningPlayerPawn))))))
